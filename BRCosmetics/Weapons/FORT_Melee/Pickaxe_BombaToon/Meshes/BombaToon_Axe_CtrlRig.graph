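# Control Rig for the BombaToon pickaxe that makes each toon eye pupil bone follow its eye's dynamic bone.
# On execution, two chained parent constraints drive the left and right dyn_eyePupil bones from dyn_eye_lf and dyn_eye_rt.
# Filters pass X/Z translation and scale but block Y translation and all rotation; the left constraint is only partially weighted, the right fully.
# The ribbon and sunglasses bones are in the hierarchy but untouched by the graph.
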ControlRigBlueprint /BRCosmetics/Weapons/FORT_Melee/Pickaxe_BombaToon/Meshes/BombaToon_Axe_CtrlRig
graphs [
 RigVMFunctionLibrary {links [], nodes []}
 RigVMModel {nodes [
   RigUnit_BeginExecution = &RigUnit_BeginExecution()
   &RigUnit_ParentConstraint(
    AdvancedSettings=(InterpolationType=Average,RotationOrderForFilter=XZY)
    Child=(Type=Bone,Name="dyn_eyePupil_lf")
    Filter=(TranslationFilter=(bX=true,bY=false,bZ=true),RotationFilter=(bX=false,bY=false,bZ=false),ScaleFilter=(bX=True,bY=True,bZ=True))
    Parents=((Item=(Type=Bone,Name="dyn_eye_lf"),Weight=1))
    Weight=0.62
    ExecutePin=@RigUnit_BeginExecution.ExecutePin
   )
   > &RigUnit_ParentConstraint(
    AdvancedSettings=(InterpolationType=Average,RotationOrderForFilter=XZY)
    Child=(Type=Bone,Name="dyn_eyePupil_rt")
    Filter=(TranslationFilter=(bX=true,bY=false,bZ=True),RotationFilter=(bX=false,bY=false,bZ=false),ScaleFilter=(bX=True,bY=True,bZ=True))
    Parents=((Item=(Type=Bone,Name="dyn_eye_rt"),Weight=1))
    Weight=1
   )
  ]}
]
hierarchy [
 b_AR_Root {loc=<0 0 0>, rot=<0 0 0.7071 0.7071>}
 Melee_TwoHanded_Handle {parent=b_AR_Root, loc=<8.1477 -0.008 -29.0793>, rot=<0 0 0.7071 -0.7071>}
 base_eye_lf {parent=b_AR_Root, loc=<9.1031 -2.9524 64.1998>, rot=<0 0 0.7071 -0.7071>}
 dyn_eye_lf {parent=base_eye_lf, loc=<0 0 0>, rot=<0 0 0 1>}
 base_eye_rt {parent=b_AR_Root, loc=<9.1031 2.9524 64.1998>, rot=<0 0 -0.7071 -0.7071>}
 dyn_eye_rt {parent=base_eye_rt, loc=<0 0 0>, rot=<0 0 0 1>}
 dyn_eyePupil_lf {parent=b_AR_Root, loc=<9.1051 -6.7436 64.1998>, rot=<-0.0071 -0.0077 0.6796 -0.7335>}
 dyn_eyePupil_rt {parent=b_AR_Root, loc=<8.9757 5.0516 64.2353>, rot=<0.0077 0.0071 0.7335 -0.6796>}
 base_ribbon01_rt {parent=b_AR_Root, loc=<0.8905 -0.2136 23.4755>, rot=<0.3389 0.0427 0.9324 0.118>}
 dyn_ribbon01_rt {parent=base_ribbon01_rt, loc=<0 0 0>, rot=<0 0 0 -1>}
 base_ribbon02_lf {parent=b_AR_Root, loc=<0.8393 -0.9896 22.3851>, rot=<-0.23 -0.0064 0.9703 -0.0743>}
 dyn_ribbon02_lf {parent=base_ribbon02_lf, loc=<0 0 0>, rot=<0 0 0 -1>}
 base_sunGlasses_cr {parent=b_AR_Root, loc=<1.3906 -0.8533 65.7624>, rot=<0 0.0026 0 -1>}
 dyn_sunGlasses_cr {parent=base_sunGlasses_cr, loc=<0 0 0>, rot=<0 0 0 -1>}
]
