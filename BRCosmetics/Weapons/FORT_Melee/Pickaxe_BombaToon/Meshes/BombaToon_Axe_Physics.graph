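# Physics asset for the BombaToon pickaxe mesh, defining physics bodies for its left and right eyes, likely for secondary jiggle motion.
# Each eye has a small capsule on its base bone and a larger, offset, rotated capsule on its dyn bone.
# One constraint per side joins dyn to base with swing and twist all free, so the set angular limit values do not restrict rotation.
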PhysicsAsset /BRCosmetics/Weapons/FORT_Melee/Pickaxe_BombaToon/Meshes/BombaToon_Axe_Physics
bodies {
 base_eye_lf [{shapeType=Capsule, length=0.5, radius=0.5, shapeName=base_eye_lf_capsule}]
 base_eye_rt [{shapeType=Capsule, length=0.5, radius=0.5, shapeName=dyn_eye_rt_capsule}]
 dyn_eye_lf [{shapeType=Capsule, length=2, radius=2, shapeName=dyn_eye_lf_capsule, center=<2 0 0>, rotation=<rot 90 0 0>}]
 dyn_eye_rt [{shapeType=Capsule, length=2, radius=2, shapeName=dyn_eye_rt_capsule, center=<2 0 0>, rotation=<rot 90 0 0>}]
}
constraints [
 {
  bone1Name=dyn_eye_rt
  bone2Name=base_eye_rt
  swing1LimitDegrees=45
  swing1Motion=Free
  swing2LimitDegrees=45
  swing2Motion=Free
  twistLimitDegrees=45
  twistMotion=Free
 }
 {
  bone1Name=dyn_eye_lf
  bone2Name=base_eye_lf
  swing1LimitDegrees=50
  swing1Motion=Free
  swing2LimitDegrees=50
  swing2Motion=Free
  twistLimitDegrees=50
  twistMotion=Free
 }
]
