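# Animation Blueprint for the BombaToon pickaxe cosmetic, adding secondary motion to its eyes, sunglasses and ribbons.
# The AnimGraph offsets the pupils, runs two RigidBody passes and a Control Rig, then springs sunglasses and ribbons with AnimDynamics.
# Update logic swaps between gameplay and front-end RigidBody passes, halving Control Rig alpha in the front end except during item preview.
# Zipline variables are set, but nothing in the AnimGraph reads them.
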AnimBlueprint /BRCosmetics/Weapons/FORT_Melee/Pickaxe_BombaToon/Meshes/BombaToon_Axe_AnimBP
target_skeleton=&/BombaToon_Axe_Skeleton
graphs [
 AnimGraph {nodes [
   AnimGraphNode_SubInput_0 = Input Pose()
   n1 = "Transform (Modify) Bone - Bone: dyn_eyePupil_lf"(ComponentPose=Local To Component(LocalPose=@AnimGraphNode_SubInput_0), Alpha=1, Translation="0.2,0,0", Rotation="0,0,0", Scale="1,1,1") {translationMode=BMM_Additive, translationSpace=BCS_BoneSpace, boneToModify {boneName=dyn_eyePupil_lf}, translation {x=0.2}}
   n2 = "Transform (Modify) Bone - Bone: dyn_eyePupil_rt"(ComponentPose=@n1, Alpha=1, Translation="-0.2,0,0", Rotation="0,0,0", Scale="1,1,1") {translationMode=BMM_Additive, translationSpace=BCS_BoneSpace, boneToModify {boneName=dyn_eyePupil_rt}, translation {x=-0.2}}
   n3 = RigidBody(ComponentPose=@n2, Alpha=Get RbanAlpha() -> (RbanAlpha:double)) {
    bOverrideWorldGravity=true
    overridePhysicsAsset=&/BombaToon_Axe_Physics
    simulationSpace=BaseBoneSpace
    baseBoneRef {boneName=base_sunGlasses_cr}
    overrideWorldGravity {z=-700}
    simSpaceSettings {maxAngularAcceleration=900, maxAngularVelocity=500, maxLinearAcceleration=900, maxLinearVelocity=500, worldAlpha=1}
   }
   n4 = RigidBody(ComponentPose=@n3, Alpha=Get RbanFEAlpha() -> (RbanFEAlpha:double)) {
    bOverrideWorldGravity=true
    overridePhysicsAsset=&BombaToon_Axe_Physics
    simulationSpace=BaseBoneSpace
    baseBoneRef {boneName=Melee_TwoHanded_Handle}
    overrideWorldGravity {z=-800}
    simSpaceSettings {maxAngularAcceleration=250, maxAngularVelocity=100, maxLinearAcceleration=250, maxLinearVelocity=100, worldAlpha=1}
   }
   n5 = Control Rig(Source=Component To Local(ComponentPose=@n4), Alpha=Get CTRL Rig Alpha() -> (CTRL Rig Alpha:double)) {controlRigAssetReference {blueprintRigClass=&/BombaToon_Axe_CtrlRig.BombaToon_Axe_CtrlRig_C}}
   n6 = Anim Dynamics - Bone: dyn_sunglasses_cr!(
    ComponentPose=Local To Component(LocalPose=@n5)
    Alpha=1
    ComponentLinearAccScale:Vector="0.3,0.3,0.3"
    ComponentLinearVelScale:Vector="0.3,0.3,0.3"
    ExternalForce="0,0,0"
   ) -> (Pose) {
    angularDampingOverride=0.9999
    angularSpringConstant=200
    bAngularSpring=true
    bOverrideAngularDamping=true
    bUseGravityOverride=true
    gravityScale=0.35
    linearSpringConstant=10000
    simulationSpace=BoneRelative
    worldSpaceMinimumScale=0
    boundBone {boneName=dyn_sunglasses_cr}
    componentAppliedLinearAccClamp=<3000 3000 3000>
    componentLinearAccScale=<0.3 0.3 0.3>
    componentLinearVelScale=<0.3 0.3 0.3>
    gravityOverride {z=-800}
    physicsBodyDefinitions [{
      boundBone {boneName=dyn_sunglasses_cr}
      boxExtents=<35 35 25>
      constraintSetup {angularLimitsMax=<0 0 0>, angularLimitsMin=<0 -18 0>, angularTarget=<1 0 0>}
      localJointOffset=<3 0 0>
     }]
    relativeSpaceBone {boneName=base_sunglasses_cr}
    simSpaceSettings {maxAngularAcceleration=60, maxAngularVelocity=80, simSpaceAngularAlpha=0.3}
   }
   n7 = Anim Dynamics - Bone: dyn_ribbon01_rt!(
    ComponentPose=@n6
    Alpha=1
    ComponentLinearAccScale:Vector="0.2,0.2,0.2"
    ComponentLinearVelScale:Vector="0.2,0.2,0.2"
    ExternalForce="0,0,0"
   ) -> (Pose) {
    angularDampingOverride=1
    angularSpringConstant=1800
    bAngularSpring=true
    bOverrideAngularDamping=true
    bUseGravityOverride=true
    gravityScale=0.35
    simulationSpace=BoneRelative
    worldSpaceMinimumScale=0.2
    boundBone {boneName=dyn_ribbon01_rt}
    componentAppliedLinearAccClamp=<20000 20000 20000>
    componentLinearAccScale=<0.2 0.2 0.2>
    componentLinearVelScale=<0.2 0.2 0.2>
    gravityOverride {z=-400}
    physicsBodyDefinitions [{
      boundBone {boneName=dyn_ribbon01_rt}
      boxExtents=<27 17 17>
      constraintSetup {angularLimitsMax=<0 45 45>, angularLimitsMin=<0 -45 -45>, angularTarget=<1 0 0>}
      localJointOffset=<2 0 0>
     }]
    relativeSpaceBone {boneName=Melee_TwoHanded_Handle}
    simSpaceSettings {maxAngularAcceleration=80, maxAngularVelocity=20, simSpaceAngularAlpha=0.65}
   }
   n8 = Anim Dynamics - Bone: dyn_ribbon02_lf!(
    ComponentPose=@n7
    Alpha=1
    ComponentLinearAccScale:Vector="0.2,0.2,0.2"
    ComponentLinearVelScale:Vector="0.2,0.2,0.2"
    ExternalForce="0,0,0"
   ) -> (Pose) {
    angularDampingOverride=1
    angularSpringConstant=1800
    bAngularSpring=true
    bOverrideAngularDamping=true
    bUseGravityOverride=true
    gravityScale=0.35
    simulationSpace=BoneRelative
    worldSpaceMinimumScale=0.2
    boundBone {boneName=dyn_ribbon02_lf}
    componentAppliedLinearAccClamp=<20000 20000 20000>
    componentLinearAccScale=<0.2 0.2 0.2>
    componentLinearVelScale=<0.2 0.2 0.2>
    gravityOverride {z=-400}
    physicsBodyDefinitions [{
      boundBone {boneName=dyn_ribbon02_lf}
      boxExtents=<27 17 17>
      constraintSetup {angularLimitsMax=<0 45 45>, angularLimitsMin=<0 -45 -45>, angularTarget=<1 0 0>}
      localJointOffset=<2 0 0>
     }]
    relativeSpaceBone {boneName=Melee_TwoHanded_Handle}
    simSpaceSettings {maxAngularAcceleration=80, maxAngularVelocity=20, simSpaceAngularAlpha=0.65}
   }
   Output Pose(Result=Component To Local(ComponentPose=@n8)) {layerGroup=None}
  ]}
 EventGraph {nodes [
   Event BlueprintUpdateAnimation()
   > n1 = Sequence()
   n2 = GetOwnerAnimationInterface!(self:Custom Character Part Anim Instance Object Reference=Self-Reference()) -> (ReturnValue:Fort Player Animation Interface Object Reference)
   > Is Valid(InputObject=@n2)
   > n3 = Set isZipline(isZipline:bool=false) -> (Output_Get:bool)
   > n4 = Branch(Condition=@n3)
   > Set animDyn on zipline(animDyn on zipline:Vector="0, 0, 0") -> (Output_Get:Vector)
   Set animDyn on zipline(execute=@n4.else, animDyn on zipline="0.3,0.3,0.3")
   n5 = Branch(execute=@n1.then_1, Condition=NOT Boolean(A=Get bIsInFrontEnd()))
   > Set RbanFEAlpha(RbanFEAlpha:double=0) -> (Output_Get:double)
   > Set RbanAlpha(RbanAlpha:double=1) -> (Output_Get:double)
   > Set CTRL Rig Alpha(CTRL Rig Alpha:double=1) -> (Output_Get:double)
   Set RbanFEAlpha(execute=@n5.else, RbanFEAlpha=1)
   > Set RbanAlpha(RbanAlpha=0)
   > n6 = Branch(Condition=Get bIsPreviewingItemInFrontEnd() -> (bIsPreviewingItemInFrontEnd:bool))
   > Set CTRL Rig Alpha(CTRL Rig Alpha=1)
   Set CTRL Rig Alpha(execute=@n6.else, CTRL Rig Alpha=0.5)
   Break Fort Anim Input Zipline!(FortAnimInput_Zipline:"Fort Anim Input Zipline Structure (by ref)"=GetZiplineInput!(self:Fort Player Animation Interface Object Reference=@n2) -> (ReturnValue:"Fort Anim Input Zipline Structure (by ref)")) -> (
    bIsZiplining:bool=~
    bIsOnAscender:bool=~
    bShouldPlayPivotTransition:bool=~
    LeanYaw:float=~
    PivotBlendDelayRemaining:float=~
    PivotBlendDelay:float=~
    PivotCardinalDirection:EFortCardinalDirection Enum
    WorldVelocityLastTick:Vector=~
   ) {is=title K2Node_BreakStruct}
  ]}
]
variables [
 animDyn on zipline {type=Vector, default=<0.3 0.3 0.3>}
 CTRL Rig Alpha {type=number, default=1}
 isZipline {type=boolean}
 RbanAlpha {type=number, default=1}
 RbanFEAlpha {type=number}
]
